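# Niagara system for the SquibblySquid character's Fowl skin that renders a pair of pulsing glow sprites on the head.
# A single burst spawns two sprites, partitioned so each follows its own skeletal mesh bone or socket with mirrored offsets.
# A system Timeline, alongside a bone-to-bone distance module, scales size and alpha, while Big Head Mode enlarges size and offsets.
# Waveform alpha varies for front end, drops one sprite during ADS and zeroes when the head is hidden.
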NiagaraSystem /BRCosmetics/Characters/Player/Male/Medium/Bodies/M_MED_SquibblySquid/Skins/Fowl/FX/NS_Character_SquibblySquid_Head
dependencies {
 usedDataInterfaces [&NiagaraDataInterfaceCurve, &NiagaraDataInterfaceSkeletalMesh]
 usedDynamicInputs [&SetBoolByFloatComparison]
 usedModules [
  &self:SystemSpawnScript.SystemScriptSource.SystemScriptGraph.NiagaraNodeAssignment_0.SetVariables_#1
  &SystemState
  &self:DistanceRangeFromBoneToBone
  &self:SystemSpawnScript.SystemScriptSource.SystemScriptGraph.NiagaraNodeAssignment_1.SetVariables_#2
  &Timeline
 ]
}
emitters {Glow {
  emitterUpdateScript [
   &EmitterState(Life Cycle Mode=::NewEnumerator1, Loop Behavior=::NewEnumerator1, Loop Duration Mode=::NewEnumerator1, UseLoopDelay=false)
   &SpawnBurst_Instantaneous(Spawn Count=2, Use Spawn Probability=false, UseLoopCountLimit=false)
  ]
  particleSpawnScript [
   &PartitionParticles()
   &InitializeParticle(
    Color=<rgba 1 0.1 1 0.2>
    Mesh Scale=<0.3 0.25 0.1>
    Mesh Scale Mode=::NewEnumerator4
    Sprite Size Mode=::NewEnumerator3
    Uniform Sprite Size=22.5
    UsePositionOffset=false
    Write Material Random=true
   )
   &InitialMeshOrientation(
    Mesh Axis To Orient=<0 0 1>
    Mesh Orientation Mode=::NewEnumerator2
    Orientation Vector=<0 1 0>
    Rotation=&MergeHelper_VectorSwitch(0=<0 -0.1 0>, 1=<0 0.1 0>, 2=?, 3=?, Index=$Particles.PartitionParticles.Partition, Num=::NewEnumerator0)
    Use Rotation Vector=true
   )
  ]
  particleUpdateScript [
   &ParticleState(Kill Particles When Lifetime Has Elapsed=false, Let Infinitely Lived Particles Die When Emitter Deactivates=true)
   &ScaleSpriteSize(
    Scale Sprite Size Mode=::NewEnumerator0
    Uniform Scale Factor=&Multiply_Float(A=$System.Timeline.NormalizedTime, B=&Lerp_Float(A=1, B=2, Alpha=$NPC.FortniteNPC.BigHeadMode))
   )
   &CameraOffset(Camera Offset Amount=&Multiply_Float(A=8, B=&Lerp_Float(A=1, B=2, Alpha=$NPC.FortniteNPC.BigHeadMode)))
   &ScaleColor(
    Scale Alpha=&Multiply_Float(
     A=&MakeCustomFloatFromBool(
      Bool=$User.bIsADS
      True Float=&MergeHelper_FloatSwitch(0=1, 1=0, 2=?, 3=?, Index=$Particles.PartitionParticles.Partition, Num=::NewEnumerator0)
      False Float=1
     )
     B=&Multiply_Float(
      A=$System.Timeline.NormalizedTime
      B=&Multiply_Float(
       A=&FN_FrontEndFloatModifier(Gameplay=1.5, Front End=1)
       B=&Waveform(
        Global Amplitude Scale=&MakeCustomFloatFromBool(Bool=&Boolean_NotOperation(Bool=$User.bisHeadHidden), True Float=1, False Float=0)
        "Waveform [1]"=::NewEnumerator12
        "[1] Input Position"=&RandomRangeFloat(Minimum=0, Maximum=1)
        "[1] Frequency"=1.5
        "[1] Amplitude Min/Max"=<0.5 1.25>
       )
      )
     )
    )
    ScaleA=true
    ScaleRGB=true
   )
   &SkeletalMeshLocation(
    Bone / Socket Index=$Particles.PartitionParticles.Partition
    Bone Sampling Mode=::NewEnumerator4
    MaskSpawnGroup=false
    Sampled Position Offset=&Multiply_VectorByFloat(
     Vector=&MergeHelper_VectorSwitch(0=<1 -3 0>, 1=<-1 -3 0>, 2=?, 3=?, Index=$Particles.PartitionParticles.Partition, Num=::NewEnumerator0)
     Float=&Lerp_Float(A=1, B=2, Alpha=$NPC.FortniteNPC.BigHeadMode)
    )
    Skeletal Mesh=$System.SkeletalMesh
   )
  ]
  rendererClasses [&NiagaraSpriteRendererProperties]
  renderers [{rendererClass=&NiagaraSpriteRendererProperties}]
 }}
summary {
 systemName=NS_Character_SquibblySquid_Head
 emitters [{emitterName=Glow, rendererClasses [&NiagaraSpriteRendererProperties]}]
 userVariables [
  User.bIsADS {type=NiagaraBool}
  User.bIsFrontEnd {type=NiagaraBool}
  User.bIsFrontEndPreview {type=NiagaraBool}
  User.bisHeadHidden {type=NiagaraBool}
 ]
}
