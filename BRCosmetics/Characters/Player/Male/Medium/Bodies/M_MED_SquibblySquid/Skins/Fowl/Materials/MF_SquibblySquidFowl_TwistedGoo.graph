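# Material function for the SquibblySquid Fowl skin that builds animated twisted goo from Mastermind's Twisted mask and normal textures.
# Both textures are panned and triplanar-projected, using pre-skinned local position and normal interpolators, with mask-based UV distortion for the normal.
# The mask blends two colors and drives optional vertex offset, while the goo normal is flattened and blended over the input normal.
# Outputs include material attributes plus individual color, normal, mask, specular, roughness and WPO.
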MaterialFunction /BRCosmetics/Characters/Player/Male/Medium/Bodies/M_MED_SquibblySquid/Skins/Fowl/Materials/MF_SquibblySquidFowl_TwistedGoo
graph {nodes [
  FunctionInput() {bUsePreviewValueAsDefault=true, inputName=----------, inputType=FunctionInput_Scalar, sortPriority=55}
  FunctionInput() {bUsePreviewValueAsDefault=true, inputName=-----------, inputType=FunctionInput_Scalar, sortPriority=105}
  FunctionInput() {bUsePreviewValueAsDefault=true, inputName=------------, inputType=FunctionInput_Scalar, sortPriority=155, previewValue=<10 0 0 1>}
  FunctionOutput(None=NamedRerouteUsage()) {outputName=MA}
  FunctionOutput(None=NamedRerouteUsage()) {outputName=NoiseMask, sortPriority=30}
  FunctionOutput(None=NamedRerouteUsage()) {outputName=Normal, sortPriority=20}
  FunctionOutput(None=NamedRerouteUsage()) {outputName=WPO, sortPriority=40}
  FunctionOutput(None=NamedRerouteUsage()) {outputName=Color, sortPriority=10}
  FunctionOutput(None=NamedRerouteUsage()) {outputName=Specular, sortPriority=11}
  FunctionOutput(None=NamedRerouteUsage()) {outputName=Roughness, sortPriority=12}
  n1 = LocalPosition() {localOrigin=InstancePreSkinning}
  NamedRerouteDeclaration(None=ShaderStageSwitch(PixelShader=VertexInterpolator(VS=@n1.XYZ).PS, VertexShader=@n1.XYZ)) {name=PreskinnedLocalPos, nodeColor=<rgba 1 0 0.0703 1>}
  n2 = PreSkinnedNormal()
  NamedRerouteDeclaration(None=ShaderStageSwitch(PixelShader=VertexInterpolator(VS=@n2).PS, VertexShader=@n2)) {name=PreskinnedLocalNormal, nodeColor=<rgba 1 0.9609 0 1>}
  n3 = Normalize(VectorInput=NamedRerouteUsage())
  NamedRerouteDeclaration(None=Multiply(A=@n3, B=NamedRerouteUsage())) {name=Pan_Normal_XY, nodeColor=<rgba 0.7109 1 0 1>}
  n4 = FunctionInput() {bUsePreviewValueAsDefault=true, inputName=GlobalTiling, inputType=FunctionInput_Scalar, sortPriority=20, previewValue=<1 0 0 1>}
  NamedRerouteDeclaration(None=@n4) {name=GlobalTiling, nodeColor=<rgba 0 0.4375 1 1>}
  n5 = NamedRerouteUsage()
  n6 = ComponentMask(None=NamedRerouteUsage()) {g=true}
  NamedRerouteDeclaration(None=Multiply(A=Multiply(A=@n5, B=@n6)) {constB=0.01}) {name=Distortion, nodeColor=<rgba 0.0078 1 0 1>}
  n7 = TextureObject() {samplerType=SAMPLERTYPE_Masks, texture=/Game/Characters/Player/Male/Medium/Bodies/M_MED_Mastermind/Textures/T_M_MED_Mastermind_Twisted_M}
  NamedRerouteDeclaration(None=@n7) {name=MaskTexture, nodeColor=<rgba 0 0.9297 1 1>}
  NamedRerouteDeclaration(None=Multiply(A=NamedRerouteUsage(), B=NamedRerouteUsage())) {name=NormalTextureSize, nodeColor=<rgba 0.6641 0 1 1>}
  NamedRerouteDeclaration(None=Multiply(A=NamedRerouteUsage(), B=NamedRerouteUsage())) {name=DistortionTextureSize, nodeColor=<rgba 1 0 0.4219 1>}
  n8 = Clamp(None=LinearInterpolate(A=NamedRerouteUsage(), B=NamedRerouteUsage(), Alpha=NamedRerouteUsage())) {clampMode=CMODE_ClampMin}
  n9 = Transform(None=NamedRerouteUsage()) {transformSourceType=TRANSFORMSOURCE_Local}
  NamedRerouteDeclaration(None=StaticSwitch(True=Multiply(A=@n8, B=@n9), False=Constant(), Value=NamedRerouteUsage())) {name=WPO, nodeColor=<rgba 0 1 0.8594 1>}
  n10 = FunctionInput() {bUsePreviewValueAsDefault=true, inputName=NormalTextureSize, inputType=FunctionInput_Scalar, sortPriority=30, previewValue=<2 0 0 1>}
  NamedRerouteDeclaration(None=@n10) {name=Normal_Size, nodeColor=<rgba 0.7812 1 0 1>}
  n11 = NamedRerouteUsage()
  n12 = NamedRerouteUsage()
  NamedRerouteDeclaration(None=MaterialFunctionCall(
    BaseNormal=NamedRerouteUsage()
    AdditionalNormal=MaterialFunctionCall(Normal=@n11, Flatness=@n12) {materialFunction=&FlattenNormal}.Result
   ) {materialFunction=&BlendAngleCorrectedNormals}.Result) {name=Normal, nodeColor=<rgba 0 1 0.5547 1>}
  n13 = NamedRerouteUsage()
  NamedRerouteDeclaration(None=LinearInterpolate(A=NamedRerouteUsage(), B=NamedRerouteUsage(), Alpha=ComponentMask(None=@n13) {b=true})) {name=BaseColor, nodeColor=<rgba 1 0.1641 0 1>}
  n14 = NamedRerouteUsage()
  n15 = NamedRerouteUsage()
  n16 = NamedRerouteUsage()
  n17 = TextureSample(UVs=Panner(Coordinate=Multiply(A=@n14, B=@n15), Speed=NamedRerouteUsage()), Tex=@n16) {samplerSource=SSM_Wrap_WorldGroupSettings}
  n18 = NamedRerouteUsage()
  n19 = Multiply(A=NamedRerouteUsage(), B=@n15)
  n20 = NamedRerouteUsage()
  n21 = Multiply(A=TextureSample(UVs=Panner(Coordinate=@n19, Speed=@n20), Tex=@n16) {samplerSource=SSM_Wrap_WorldGroupSettings}.RGB, B=NamedRerouteUsage())
  n22 = Multiply(A=NamedRerouteUsage(), B=@n15)
  n23 = NamedRerouteUsage()
  n24 = Multiply(A=TextureSample(UVs=Panner(Coordinate=@n22, Speed=@n23), Tex=@n16) {samplerSource=SSM_Wrap_WorldGroupSettings}.RGB, B=NamedRerouteUsage())
  NamedRerouteDeclaration(None=Add(A=Multiply(A=@n17.RGB, B=@n18), B=Add(A=@n21, B=@n24))) {name=Mask, nodeColor=<rgba 1 0.8672 0 1>}
  NamedRerouteDeclaration(None=FunctionInput(Preview=Constant3Vector() {constant=<rgba 0 0 1 0>}) {bUsePreviewValueAsDefault=true, inputName=Original Normal}) {name=IN_Normal, nodeColor=<rgba 0 1 0.4141 1>}
  n25 = SetMaterialAttributes!(
   MaterialAttributes
   BaseColor=NamedRerouteUsage()
   Specular=NamedRerouteUsage()
   Roughness=NamedRerouteUsage()
   Metallic=NamedRerouteUsage()
   Normal=NamedRerouteUsage()
   WorldPositionOffset=NamedRerouteUsage()
  ) {attributeSetTypes [
    #1
    #2
    #3
    #4
    #5
    #6
   ]}
  NamedRerouteDeclaration(None=@n25) {name=OUT_Material, nodeColor=<rgba 0 1 1 1>}
  n26 = TextureObject() {samplerType=SAMPLERTYPE_Normal, texture=/Game/Characters/Player/Male/Medium/Bodies/M_MED_Mastermind/Textures/T_M_MED_Mastermind_Twisted_N}
  NamedRerouteDeclaration(None=@n26) {name=NormalTexture, nodeColor=<rgba 1 0 0.2812 1>}
  n27 = FunctionInput() {bUsePreviewValueAsDefault=true, inputName=DistortionPanSpeed, inputType=FunctionInput_Scalar, sortPriority=90, previewValue=<0.04 0 0 1>}
  NamedRerouteDeclaration(None=@n27) {name=PanSpeedMask, nodeColor=<rgba 1 0 0.375 1>}
  n28 = FunctionInput() {bUsePreviewValueAsDefault=true, inputName=NormalPanSpeed, inputType=FunctionInput_Scalar, sortPriority=100, previewValue=<0.06 0 0 1>}
  NamedRerouteDeclaration(None=@n28) {name=PanSpeedNormal, nodeColor=<rgba 0 1 1 1>}
  n29 = FunctionInput() {bUsePreviewValueAsDefault=true, inputName=DistortionTextureSize, inputType=FunctionInput_Scalar, sortPriority=40, previewValue=<4 0 0 1>}
  NamedRerouteDeclaration(None=@n29) {name=Dist_Size, nodeColor=<rgba 0.8984 0 1 1>}
  n30 = FunctionInput() {
   bUsePreviewValueAsDefault=true
   inputName=PanDirection YZ
   inputType=FunctionInput_Vector2
   sortPriority=60
   previewValue=<-0.0857 0.1333 0 1>
  }
  NamedRerouteDeclaration(None=@n30) {name=PanDirection YZ, nodeColor=<rgba 0 1 0.5781 1>}
  n31 = FunctionInput() {
   bUsePreviewValueAsDefault=true
   inputName=PanDirection XZ
   inputType=FunctionInput_Vector2
   sortPriority=70
   previewValue=<0.0762 0.1333 0 1>
  }
  NamedRerouteDeclaration(None=@n31) {name=PanDirection XZ, nodeColor=<rgba 0 1 0.5781 1>}
  n32 = FunctionInput() {bUsePreviewValueAsDefault=true, inputName=PanDirection XY, inputType=FunctionInput_Vector2, sortPriority=80, previewValue=<0.0286 -0.05 0 1>}
  NamedRerouteDeclaration(None=@n32) {name=PanDirection XY, nodeColor=<rgba 0 1 0.5781 1>}
  n33 = FunctionInput() {
   bUsePreviewValueAsDefault=true
   inputName=Normal Distortion Strength
   inputType=FunctionInput_Scalar
   sortPriority=50
   previewValue=<6 0 0 1>
  }
  NamedRerouteDeclaration(None=@n33) {name=NormalDistortionStrength, nodeColor=<rgba 0.3828 1 0 1>}
  n34 = FunctionInput() {bUsePreviewValueAsDefault=true, inputName=VertexOffsetMin, inputType=FunctionInput_Scalar, sortPriority=160, previewValue=<-0.4238 0 0 1>}
  NamedRerouteDeclaration(None=@n34) {name=VertexOffset_Min, nodeColor=<rgba 1 0 0.7031 1>}
  n35 = FunctionInput() {bUsePreviewValueAsDefault=true, inputName=VertexOffsetMax, inputType=FunctionInput_Scalar, sortPriority=170, previewValue=<0.791 0 0 1>}
  NamedRerouteDeclaration(None=@n35) {name=VertexOffset_Max, nodeColor=<rgba 1 0 0.7031 1>}
  n36 = FunctionInput() {bUsePreviewValueAsDefault=true, inputName=VertexOffsetDirection, sortPriority=180, previewValue=<2 -1 2 1>}
  NamedRerouteDeclaration(None=@n36) {name=VertexOffsetDirection, nodeColor=<rgba 1 0.6563 0 1>}
  n37 = FunctionInput() {
   bUsePreviewValueAsDefault=true
   inputName=NormalProjectionContrast
   inputType=FunctionInput_Scalar
   sortPriority=151
   previewValue=<0.5 0 0 1>
  }
  NamedRerouteDeclaration(None=@n37) {name=NormalProjectionContrast, nodeColor=<rgba 0.3359 0 1 1>}
  NamedRerouteDeclaration(None=FunctionInput() {bUsePreviewValueAsDefault=true, inputName=Color 01, sortPriority=110}) {name=Color_01, nodeColor=<rgba 0 1 0.0156 1>}
  n38 = FunctionInput() {bUsePreviewValueAsDefault=true, inputName=Color 02, sortPriority=120, previewValue=<0.0022 0.0017 0.005 1>}
  NamedRerouteDeclaration(None=@n38) {name=Color_02, nodeColor=<rgba 0 1 0.0156 1>}
  MaterialExpressionFunctionInput_17 = FunctionInput() {bUsePreviewValueAsDefault=true, inputName=SRM, sortPriority=130, previewValue=<0.3238 0.33 0 1>}
  n39 = Convert(None=@MaterialExpressionFunctionInput_17) {
   nodeName=Break Float3
   convertInputs [{
     type=Vector3
     expressionInput {
      expression=&self:MaterialExpressionFunctionInput_17
      inputName=None
      mask=0
      maskA=0
      maskB=0
      maskG=0
      maskR=0
      outputIndex=0
     }
    }]
   convertOutputs [{type=Scalar}, {type=Scalar}, {type=Scalar}]
  }
  NamedRerouteDeclaration(None=@n39) {name=Specular, nodeColor=<rgba 0 0.625 1 1>}
  n40 = Normalize(VectorInput=NamedRerouteUsage())
  NamedRerouteDeclaration(None=Multiply(A=NamedRerouteUsage(), B=@n40)) {name=Pan_Mask_YZ, nodeColor=<rgba 0.7109 1 0 1>}
  NamedRerouteDeclaration(None=@n39) {name=Roughness, nodeColor=<rgba 0.9453 1 0 1>}
  NamedRerouteDeclaration(None=@n39) {name=Metallic, nodeColor=<rgba 0.9453 1 0 1>}
  NamedRerouteDeclaration(None=FunctionInput() {bUsePreviewValueAsDefault=true, inputName=Normal Flatten, inputType=FunctionInput_Scalar, sortPriority=150}) {name=NormalFlatten, nodeColor=<rgba 1 0.9141 0 1>}
  NamedRerouteDeclaration(None=FunctionInput(Preview=StaticBool()) {bUsePreviewValueAsDefault=true, inputName=UseWPO, inputType=FunctionInput_StaticBool, sortPriority=190}) {name=UseWPOlogic, nodeColor=<rgba 0.7578 1 0 1>}
  n41 = NamedRerouteUsage()
  n42 = Abs(None=Normalize(VectorInput=NamedRerouteUsage()))
  n43 = ComponentMask(None=@n42) {r=true}
  n44 = ComponentMask(None=@n42) {g=true}
  n45 = ComponentMask(None=@n42) {b=true}
  n46 = Max(A=@n43, B=Max(A=@n44, B=@n45))
  n47 = Subtract(A=@n46, B=@n43)
  n48 = AppendVector(A=Subtract(A=@n46, B=@n44), B=Subtract(A=@n46, B=@n45))
  n49 = Saturate(None=InverseLinearInterpolate(A=Add(A=@n41) {constB=0.01}, Value=AppendVector(A=@n47, B=@n48)) {constA=0.2, constB=0})
  n50 = ComponentMask(None=@n49) {r=true}
  n51 = ComponentMask(None=@n49) {g=true}
  n52 = ComponentMask(None=@n49) {b=true}
  n53 = Add(A=@n50, B=Add(A=@n51, B=@n52))
  n54 = Saturate(None=Divide(A=@n50, B=@n53))
  n55 = NamedRerouteDeclaration(None=Divide(A=@n51, B=@n53)) {name=Y Projection Weight, nodeColor=<rgba 0.5 0.5 0 1>}
  n56 = NamedRerouteDeclaration(None=Divide(A=@n52, B=@n53)) {name=Z Projection Weight, nodeColor=<rgba 0.5 0.5 0 1>}
  NamedRerouteDeclaration(None=AppendVector(A=NamedRerouteDeclaration(None=@n54) {name=X Projection Weight, nodeColor=<rgba 0.5 0.5 0 1>}, B=AppendVector(A=@n55, B=@n56))) {name=Projection Weight, nodeColor=<rgba 0 1 0.6484 1>}
  n57 = Normalize(VectorInput=VertexNormalWS())
  NamedRerouteDeclaration(None=Sign(None=NamedRerouteDeclaration(None=@n57) {name=VertexNormal, nodeColor=<rgba 0.25 0.25 1 1>})) {name=ProjectionSigns, nodeColor=<rgba 1 1 0 1>}
  n58 = Divide(A=NamedRerouteUsage()) {constB=400}
  n59 = ComponentMask(None=@n58) {g=true, r=true}
  n60 = AppendVector(A=Constant() {r=1}, B=ComponentMask(None=NamedRerouteUsage()) {b=true})
  n61 = NamedRerouteUsage()
  NamedRerouteDeclaration(None=Multiply(A=Multiply(A=@n59, B=@n60), B=@n61)) {name=Z Projection UV, nodeColor=<rgba 1 0.25 0 1>}
  n62 = ComponentMask(None=@n58) {b=true, r=true}
  n63 = AppendVector(A=ComponentMask(None=NamedRerouteUsage()) {g=true}, B=Constant() {r=-1})
  NamedRerouteDeclaration(None=Multiply(A=Multiply(A=@n62, B=@n63), B=@n61)) {name=Y Projection UV, nodeColor=<rgba 1 0.25 0 1>}
  NamedRerouteDeclaration(None=Multiply(A=@n40, B=NamedRerouteUsage())) {name=Pan_Normal_YZ, nodeColor=<rgba 0.7109 1 0 1>}
  n64 = ComponentMask(None=@n58) {b=true, g=true}
  n65 = NamedRerouteUsage()
  n66 = AppendVector(A=Multiply(A=ComponentMask(None=@n65) {r=true}) {constB=-1}, B=Constant() {r=-1})
  NamedRerouteDeclaration(None=Multiply(A=Multiply(A=@n64, B=@n66), B=@n61)) {name=X Projection UV, nodeColor=<rgba 1 0.25 0 1>}
  NamedRerouteDeclaration(None=Constant3Vector() {constant=<rgba 0 1 0 0>}) {name=World Y Axis, nodeColor=<rgba 0 1 0 1>}
  NamedRerouteDeclaration(None=Constant3Vector() {constant=<rgba 1 0 0 0>}) {name=World X Axis, nodeColor=<rgba 1 0 0 1>}
  NamedRerouteDeclaration(None=Constant3Vector() {constant=<rgba 0 0 1 0>}) {name=World Z Axis, nodeColor=<rgba 0 0 1 1>}
  n67 = NamedRerouteUsage()
  n68 = NamedRerouteUsage()
  n69 = NamedRerouteUsage()
  n70 = Panner(Coordinate=Multiply(A=Add(A=@n67, B=@n68), B=@n69), Speed=NamedRerouteUsage())
  n71 = NamedRerouteUsage()
  n72 = ComponentMask(None=NamedRerouteUsage()) {r=true}
  MaterialExpressionCrossProduct_6 = CrossProduct(A=NamedRerouteUsage(), B=Multiply(A=NamedRerouteUsage(), B=Multiply(A=@n72) {constB=-1}))
  n73 = Custom!(v=@MaterialExpressionCrossProduct_6) {
   code="float3 Y = float3(0.f,1.f,0.f);\r\nreturn (dot(v,v)!=0.f)?v:Y;"
   description=Y Nan Check
   inputs [{
     inputName=v
     input {expression=&self:MaterialExpressionCrossProduct_6}
    }]
  }
  n74 = Normalize(VectorInput=@n73)
  n75 = NamedRerouteUsage()
  n76 = Multiply(
   A=MaterialFunctionCall(
    VectorToTransform=TextureSample(UVs=@n70, Tex=@n71) {samplerSource=SSM_Wrap_WorldGroupSettings, samplerType=SAMPLERTYPE_Normal}.RGB
    BasisX=CrossProduct(A=@n74, B=@n75)
    BasisY=@n74
    BasisZ=NamedRerouteUsage()
   ) {desc=TangentToWorld, materialFunction=&Transform3x3Matrix}.Transformed Vector
   B=NamedRerouteUsage()
  )
  n77 = Add(A=NamedRerouteUsage(), B=@n68)
  n78 = TextureSample(UVs=Panner(Coordinate=Multiply(A=@n77, B=@n69), Speed=NamedRerouteUsage()), Tex=@n71) {samplerSource=SSM_Wrap_WorldGroupSettings, samplerType=SAMPLERTYPE_Normal}
  n79 = NamedRerouteUsage()
  MaterialExpressionCrossProduct_4 = CrossProduct(A=NamedRerouteUsage(), B=Multiply(A=NamedRerouteUsage(), B=ComponentMask(None=@n79) {g=true}))
  n80 = Custom(v=@MaterialExpressionCrossProduct_4) {
   code="float3 X = float3(1.f,0.f,0.f);\r\nreturn (dot(v,v)!=0.f)?v:X;"
   description=X Nan Check
   inputs [{
     inputName=v
     input {expression=&self:MaterialExpressionCrossProduct_4}
    }]
  }
  n81 = Normalize(VectorInput=@n80)
  n82 = CrossProduct(A=@n81, B=NamedRerouteUsage())
  n83 = NamedRerouteUsage()
  n84 = Add(A=NamedRerouteUsage(), B=@n68)
  n85 = TextureSample(UVs=Panner(Coordinate=Multiply(A=@n84, B=@n69), Speed=NamedRerouteUsage()), Tex=@n71) {samplerSource=SSM_Wrap_WorldGroupSettings, samplerType=SAMPLERTYPE_Normal}
  MaterialExpressionCrossProduct_2 = CrossProduct(A=NamedRerouteUsage(), B=NamedRerouteUsage())
  n86 = Custom(v=@MaterialExpressionCrossProduct_2) {
   code="float3 X = float3(1.f,0.f,0.f);\r\nreturn (dot(v,v)!=0.f)?v:X;"
   description=X Nan Check
   inputs [{
     inputName=v
     input {expression=&self:MaterialExpressionCrossProduct_2}
    }]
  }
  n87 = Normalize(VectorInput=@n86)
  n88 = CrossProduct(A=@n87, B=NamedRerouteUsage())
  n89 = NamedRerouteUsage()
  n90 = Add(
   A=Multiply(
    A=MaterialFunctionCall(VectorToTransform=@n78.RGB, BasisX=@n82, BasisY=@n81, BasisZ=@n83) {desc=TangentToWorld, materialFunction=&Transform3x3Matrix}.Transformed Vector
    B=NamedRerouteUsage()
   )
   B=Multiply(
    A=MaterialFunctionCall(VectorToTransform=@n85.RGB, BasisX=@n88, BasisY=@n87, BasisZ=@n89) {desc=TangentToWorld, materialFunction=&Transform3x3Matrix}.Transformed Vector
    B=NamedRerouteUsage()
   )
  )
  NamedRerouteDeclaration(None=Transform(None=Add(A=@n76, B=@n90)) {transformSourceType=TRANSFORMSOURCE_World, transformType=TRANSFORM_Tangent}) {name=TriplannarNormal, nodeColor=<rgba 0 1 0.5078 1>}
  n91 = NamedRerouteUsage()
  n92 = NamedRerouteUsage()
  n93 = NamedRerouteUsage()
  n94 = TextureSample(UVs=Panner(Coordinate=Multiply(A=@n91, B=@n92), Speed=NamedRerouteUsage()), Tex=@n93) {samplerSource=SSM_Wrap_WorldGroupSettings}
  n95 = NamedRerouteUsage()
  n96 = Multiply(A=NamedRerouteUsage(), B=@n92)
  n97 = NamedRerouteUsage()
  n98 = Multiply(A=TextureSample(UVs=Panner(Coordinate=@n96, Speed=@n97), Tex=@n93) {samplerSource=SSM_Wrap_WorldGroupSettings}.B, B=NamedRerouteUsage())
  n99 = Multiply(A=NamedRerouteUsage(), B=@n92)
  n100 = NamedRerouteUsage()
  n101 = Multiply(A=TextureSample(UVs=Panner(Coordinate=@n99, Speed=@n100), Tex=@n93) {samplerSource=SSM_Wrap_WorldGroupSettings}.B, B=NamedRerouteUsage())
  NamedRerouteDeclaration(None=Add(A=Multiply(A=@n94.B, B=@n95), B=Add(A=@n98, B=@n101))) {name=WPOMask, nodeColor=<rgba 1 0.8672 0 1>}
  n102 = Normalize(VectorInput=NamedRerouteUsage())
  NamedRerouteDeclaration(None=Multiply(A=NamedRerouteUsage(), B=@n102)) {name=Pan_Mask_XZ, nodeColor=<rgba 0.7109 1 0 1>}
  NamedRerouteDeclaration(None=Multiply(A=NamedRerouteUsage(), B=@n3)) {name=Pan_Mask_XY, nodeColor=<rgba 0.7109 1 0 1>}
  NamedRerouteDeclaration(None=Multiply(A=@n102, B=NamedRerouteUsage())) {name=Pan_Normal_XZ, nodeColor=<rgba 0.7109 1 0 1>}
 ]}
