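# Wrapper material function that layers the animated TwistedGoo effect onto an input material for the Squibbly Squid Fowl skin.
# It feeds the incoming normal and shared "TwistGoo" parameters for three-plane panning, distortion, color and vertex offset into MF_SquibblySquidFowl_TwistedGoo.
# A channel-selectable mask, from an input or optional texture, drives the attribute blend.
# The blend keeps the base vertex attributes, while goo WPO and other channels have separate outputs.
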MaterialFunction /BRCosmetics/Characters/Player/Male/Medium/Bodies/M_MED_SquibblySquid/Skins/Fowl/Materials/MFW_SquibblySquidFowl_TwistedGoo
graph {nodes [
  n1 = GetMaterialAttributes(None=FunctionInput(Preview=SetMaterialAttributes!(MaterialAttributes)) {bUsePreviewValueAsDefault=true, inputType=FunctionInput_MaterialAttributes}) {attributeGetTypes [#1]}
  n2 = ScalarParameter() {defaultValue=4, group=TwistGoo - Global Parameters, parameterName=TwistGoo.NormalTextureSize, primitiveDataIndex=10, sortPriority=20}
  n3 = VectorParameter() {group=TwistGoo - Global Parameters, parameterName=TwistGoo.PanDirection_YZ, sortPriority=50, defaultValue=<rgba -0.0857 0.1333 0 1>}
  n4 = VectorParameter() {group=TwistGoo - Global Parameters, parameterName=TwistGoo.PanDirection_XZ, sortPriority=60, defaultValue=<rgba 0.0762 0.1333 0 1>}
  n5 = VectorParameter() {group=TwistGoo - Global Parameters, parameterName=TwistGoo.PanDirection_XY, sortPriority=70, defaultValue=<rgba 0.0286 -0.05 0 1>}
  n6 = VectorParameter() {group=TwistGoo - Global Parameters, parameterName=TwistGoo.Color_01, sortPriority=100, defaultValue=<rgba 0 0 0 1>}
  n7 = VectorParameter() {group=TwistGoo - Global Parameters, parameterName=TwistGoo.Color_02, sortPriority=105, defaultValue=<rgba 0.0022 0.0017 0.005 1>}
  n8 = VectorParameter() {group=TwistGoo - Global Parameters, parameterName=TwistGoo.SRM, sortPriority=110, defaultValue=<rgba 0.3238 0.33 0 1>}
  n9 = VectorParameter() {group=TwistGoo - Global Parameters, parameterName=TwistGoo.VertexOffsetDirection, sortPriority=150, defaultValue=<rgba 2 -1 2 1>}
  n10 = MaterialFunctionCall!(
   Original Normal=@n1.Normal
   GlobalTiling=ScalarParameter() {defaultValue=1, group=TwistGoo - Global Parameters, parameterName=TwistGoo.GlobalTiling, sortPriority=10}
   NormalTextureSize=@n2
   DistortionTextureSize=ScalarParameter() {defaultValue=4, group=TwistGoo - Global Parameters, parameterName=TwistGoo.DistortionTextureSize, sortPriority=30}
   Normal Distortion Strength=ScalarParameter() {defaultValue=2, group=TwistGoo - Global Parameters, parameterName=TwistGoo.NormalDistortionStrength, sortPriority=40}
   ----------
   PanDirection YZ=ComponentMask(None=@n3.RGB) {g=true, r=true}
   PanDirection XZ=ComponentMask(None=@n4.RGB) {g=true, r=true}
   PanDirection XY=ComponentMask(None=@n5.RGB) {g=true, r=true}
   DistortionPanSpeed=ScalarParameter() {defaultValue=0.04, group=TwistGoo - Global Parameters, parameterName=TwistGoo.DistortionPanSpeed, sortPriority=80}
   NormalPanSpeed=ScalarParameter() {defaultValue=0.02, group=TwistGoo - Global Parameters, parameterName=TwistGoo.NormalPanSpeed, sortPriority=90}
   -----------
   Color 01=@n6.RGB
   Color 02=@n7.RGB
   SRM=@n8.RGB
   Normal Flatten=ScalarParameter() {defaultValue=0.25, group=TwistGoo - Global Parameters, parameterName=TwistGoo.NormalFlatten, sortPriority=119}
   NormalProjectionContrast=ScalarParameter() {defaultValue=0.1, group=TwistGoo - Global Parameters, parameterName=TwistGoo.NormalProjectionContrast, sortPriority=120}
   ------------
   VertexOffsetMin=ScalarParameter() {defaultValue=-0.4238, group=TwistGoo - Global Parameters, parameterName=TwistGoo.VertexOffsetMin, sortPriority=130}
   VertexOffsetMax=ScalarParameter() {defaultValue=0.791, group=TwistGoo - Global Parameters, parameterName=TwistGoo.VertexOffsetMax, sortPriority=140}
   VertexOffsetDirection=@n9.RGB
   UseWPO=StaticBoolParameter() {group=TwistGoo - Global Parameters, parameterName=TwistGoo.UseWPO, sortPriority=0}
  ) {materialFunction=&/MF_SquibblySquidFowl_TwistedGoo}
  FunctionOutput(None=BlendMaterialAttributes(A=NamedRerouteUsage(), B=@n10.MA, Alpha=NamedRerouteUsage()) {vertexAttributeBlendType=UseA}) {outputName=Blended MA}
  FunctionOutput(None=@n10.NoiseMask) {outputName=TwistedGoo NoiseMask, sortPriority=30}
  FunctionOutput(None=@n10.Normal) {outputName=TwistedGoo Normal, sortPriority=20}
  FunctionOutput(None=@n10.WPO) {outputName=TwistedGoo WPO, sortPriority=40}
  FunctionOutput(None=@n10.Color) {outputName=TwistedGoo Color, sortPriority=10}
  FunctionOutput(None=@n10.Specular) {outputName=TwistedGoo Specular, sortPriority=11}
  FunctionOutput(None=@n10.Roughness) {outputName=TwistedGoo Roughness, sortPriority=12}
  FunctionOutput(None=@n10.MA) {outputName=TwistedGoo MA, sortPriority=1}
  NamedRerouteDeclaration(None=@n1.MaterialAttributes) {name=IN_Material, nodeColor=<rgba 0 1 0.9062 1>}
  n11 = TextureSampleParameter2D() {
   group=TwistGoo - Global Parameters
   parameterName=TwistGoo.CustomTAMask
   samplerSource=SSM_Wrap_WorldGroupSettings
   samplerType=SAMPLERTYPE_Alpha
   sortPriority=1
   texture=/Game/Global/Textures/Default/Blanks/T_BlackAlpha
  }
  n12 = FunctionInput(Preview=Constant4Vector() {constant=<rgba 1 0 0 0>}.RGBA) {bUsePreviewValueAsDefault=true, inputName=TA Mask, sortPriority=10, previewValue=<1 0 0 1>}
  NamedRerouteDeclaration(None=ChannelMaskParameter(None=StaticSwitchParameter(True=@n11.RGB, False=@n12) {group=TwistGoo - Global Parameters, parameterName=TwistGoo.UseCustomTAMask, sortPriority=0}) {group=TwistGoo - Global Parameters, parameterName=TwistGoo.TAMaskChannel, sortPriority=1}.R) {name=TAMask, nodeColor=<rgba 1 0 0.3984 1>}
 ]}
